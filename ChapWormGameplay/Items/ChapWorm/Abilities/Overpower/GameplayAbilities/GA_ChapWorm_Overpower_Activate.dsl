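; Gameplay ability that activates ChapWorm Overpower, spending the kinetic energy charge stored as the source weapon's magazine ammo.
; Activation waits out a temporary blocking tag, then starts a hotfixable timeout, plays the activate montage and sends a tutorial event.
; At the montage notify, a full charge applies the Overpower status effect, and the server deducts ammo.
; Helper functions handle the failure cue and which failure reasons play a failed montage.
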(event Ability|EventActivateAbility
  (bind _returnvalue (Ability|GetCurrentSourceWeapon))
  (bind _returnvalue_2 (FortWeapon|GetMagazineAmmoCount _returnvalue))
  (bind _temporaryactivationblockedtag (Variables|Tags|GetTemporaryActivationBlockedTag))
  ;[0.04753 0.1667 0.0738] ActivateAbility: Play the activate montage and apply the surge gameplay effect at the designated notify.
  ;; Queue Activation if Owner has Tag
  (if (GameplayTags|HasTag (GameplayTags|GetOwnedGameplayTags (Ability|GetAbilitySystemComponentfromActorInfo) :!class /Script/GameplayTags.BlueprintGameplayTagLibrary) _temporaryactivationblockedtag true)
    (Ability|Tasks|WaitGameplayTagRemove _temporaryactivationblockedtag 0 true
      (:Removed
        _)))
  (|CallParentFunction)
  (bind _asynctaskproxy (GameplayTasks|TaskWaitDelay 0 (ScalableFloat|GetValueatLevel (Variables|Hotfix|GetHotfix_TimeOutDuration)))
    (:then
      (Variables|Transient|SetAsyncTask_TimeOutDelay _asynctaskproxy)
      (Utilities|FlowControl|DelayUntilNextTick)
      (bind (_asynctaskproxy_2 _notifytag) (Ability|Tasks|PlayMontageAndWaitNotify "KineticEnergyActivate" "/ChapWormGameplay/Items/ChapWorm/Animation/KineticEnergy/ChapWorm_KineticEnergy_Activate_M.ChapWorm_KineticEnergy_Activate_M")
        (:then
          (Variables|Transient|SetAsyncTask_PlayMontage _asynctaskproxy_2)
          ;; Tutorial complete.
          (Ability|SendGameplayEventtoActor (CallFunction|GetAvatarasFortPlayerPawn) "(TagName=\"ChapWorm.Tutorial.Overpower.Activate.End\")"))
        (:OnCompleted
          _)
        (:OnBlendOut
          _)
        (:OnInterrupted
          _)
        (:OnCancelled
          _)
        (:OnNotifyBegin
          (if (== _notifytag "(TagName=\"ChapWorm.AnimNotify.Overpower.Activate\")")
            (Utilities|IsValid _returnvalue
              (:Is%20Valid
                (if (>= _returnvalue_2 100)
                  (Fortnite|FortShippingLog self "ActivateAbility: Success.")
                  (Ability|ApplyGameplayEffectToOwner "/ChapWormGameplay/Items/ChapWorm/Abilities/Overpower/GameplayEffects/GE_ChapWorm_ApplyOverpowerStatus.GE_ChapWorm_ApplyOverpowerStatus_C")
                  (Ability|HasAuthority
                    (:True
                      (FortWeapon|ModAmmo (Ability|GetCurrentSourceWeapon) -99)))
                  (else
                    ;;[1 0 0] Warning: Charge not at max
                    (Fortnite|FortShippingLog self (Utilities|String|Append "ActivateAbility: Charge not at max: " (Utilities|String|ToString(Integer) _returnvalue_2)) true))))
              (:Is%20Not%20Valid
                ;;[1 0 0] Warning: SourceWeapon is null.
                (Fortnite|FortShippingLog self "ActivateAbility: SourceWeapon is null." true)))
            (Ability|EndAbility))))
      (Ability|EndAbility))
    (:OnFinish
      (Ability|EndAbility))))

(event Ability|EventOnEndAbility
  (params (bWasCancelled :type Boolean))
  ;[0.1094 0.07824 0.02183 0.5] EndAbility: Clean up.
  (Utilities|FlowControl|Sequence
    (:then_0
      (GameplayTasks|EndTask (Variables|Transient|GetAsyncTask_TimeOutDelay)))
    (:then_1
      (|CallParentFunction bWasCancelled))))

(fn ExecuteFailGameplayCue
  (params (Target :type Actor%20Object%20Reference))
  (GameplayCue|ExecuteGameplayCue%28Burst%29Local Target (Variables|Settings|GetGameplayCueTagActivateFailed) (Ability|GameplayCue|MakeGameplayCueParameters 0.0)))

(fn OverrideFailedReason
  (params (FailedReason :type Gameplay%20Tag%20Container%20Structure%20%28by%20ref%29))
  (return FailedReason))

(fn ShouldPlayFailedMontage
  (params (FailedReason :type Gameplay%20Tag%20Container%20Structure%20%28by%20ref%29))
  (return (GameplayTags|HasAnyTags FailedReason (GameplayTags|MakeLiteralGameplayTagContainer "(GameplayTags=((TagName=\"Athena.Ability.Failed.ChapWorm.Overpower.Activate.AlreadyActive\"),(TagName=\"Athena.Ability.Failed.ChapWorm.Overpower.Activate.NotEnoughKineticEnergy\")))"))))
