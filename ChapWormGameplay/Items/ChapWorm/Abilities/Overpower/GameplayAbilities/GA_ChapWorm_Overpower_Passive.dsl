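; Passive gameplay ability for the ChapWorm Overpower state, handling its camera modifier and duration refresh.
; On activation it caches the avatar pawn, adds the Overpower camera modifier and binds to the pawn's damage event.
; ChapWorm weapon hits that drop another target's health to zero can apply the refresh effect, gated by hotfix toggles for DBNO and elimination.
; On end it winds down the camera and starts the kinetic energy charge tutorial through an event and actor tag.
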(event Ability|EventActivateAbility
  ;[0.02847 0.15 0.05885 0.5] ActivateAbility: Cache the player pawn.
  (|CallParentFunction)
  (bind _outfortplayerpawn (CallFunction|GetAvatarasValidFortPlayerPawn)
    (:ValidFortPlayerPawn
      (Variables|Transient|SetAvatarPlayerPawn _outfortplayerpawn)
      (Default|AddCameraModifier)
      ;[0.02285 0.1074 0.1198 0.5] Bind to know when the player pawn does damage.
      (EventDispatchers|BindEventtoOnDidDamage (Variables|Transient|GetAvatarPlayerPawn) (EventDispatchers|CreateEvent :!event Custom|HandleAvatarPlayerPawnDidDamage)))))

(event Ability|EventOnEndAbility
  (params (bWasCancelled :type Boolean))
  ;[0.15 0.08281 0.009046 0.5] EndAbility: When the overpower state has ended, start a tutorial to let players know how to charge up their kinetic energy again.
  (Utilities|FlowControl|Sequence
    (:then_0
      (Default|RemoveCameraModifier)
      (bind _avatarplayerpawn (Variables|Transient|GetAvatarPlayerPawn)
        (:then
          (Ability|SendGameplayEventtoActor _avatarplayerpawn "(TagName=\"ChapWorm.Tutorial.KineticEnergy.Charge.Start\")")
          (Utilities|Array|AddUnique (Variables|/Script/Engine.Actor|GetTags _avatarplayerpawn) "ChapWormTutorialKineticEnergyChargeActive"))
        (:else)))
    (:then_1
      (|CallParentFunction bWasCancelled))))

(event Custom|HandleAvatarPlayerPawnDidDamage
  (params (DamagedActor :type Actor%20Object%20Reference) (Damage :type Float%20%28single-precision%29) (InstigatedBy :type Controller%20Object%20Reference) (DamageCauser :type Actor%20Object%20Reference) (HitLocation :type Vector) (FHitComponent :type Primitive%20Component%20Object%20Reference) (BoneName :type Name) (Momentum :type Vector))
  (bind (_bsuccessfullyfoundattribute _returnvalue) (Ability|Attribute|GetFloatAttribute DamagedActor "(AttributeName=\"Health\",Attribute=/Script/FortniteGame.FortHealthSet:Health,AttributeOwner=\"/Script/CoreUObject.Class'/Script/FortniteGame.FortHealthSet'\")"))
  (bind (_asb_weapon_chap_worm _bsuccess) (Utilities|Casting|CastToB_Weapon_ChapWorm DamageCauser))
  ;[0.15 0.04864 0.1285 0.5] HandleAvatarPlayerPawnDidDamage: Refresh the overpower duration, if desired.
  (bind _asfort_pawn (Utilities|Casting|CastToFortPawn DamagedActor))
  (if (and (!= _asfort_pawn (Variables|Transient|GetAvatarPlayerPawn)) _bsuccess)
    (if _bsuccessfullyfoundattribute
      (if (<= _returnvalue 0)
        (if (Health|IsDBNO _asfort_pawn)
          (if (ScalableFloat|GetValueasBool (Variables|Hotfix|GetHotfix_RefreshOnPawnEliminated))
            _)
          (else
            (if (ScalableFloat|GetValueasBool (Variables|Hotfix|GetHotfix_RefreshOnPawnDBNOd))
              _)))
        (Ability|ApplyGameplayEffectToOwner "/ChapWormGameplay/Items/ChapWorm/Abilities/Overpower/GameplayEffects/GE_ChapWorm_OverpowerRefresh.GE_ChapWorm_OverpowerRefresh_C")))))

(fn AddCameraModifier
  ;[0.000237 0.04228 0.05729] AddCameraMode: Add the desired camera mode to the player pawn.
  (bind _outfortplayerpawn (CallFunction|GetAvatarasValidFortPlayerPawn)
    (:ValidFortPlayerPawn
      (Utilities|IsValid (Fortnite|GetFortPlayerControllerfromActor _outfortplayerpawn)
        (:Is%20Valid
          (bind _playercameramanager (Variables|/Script/Engine.PlayerController|GetPlayerCameraManager (Fortnite|GetFortPlayerControllerfromActor _outfortplayerpawn))
            (:then
              (bind _asfort_player_camera_base (Utilities|Casting|CastToFortPlayerCameraBase _playercameramanager))
              (bind _returnvalue (Game|Player|AddNewCameraModifier _asfort_player_camera_base "/ChapWormGameplay/Items/ChapWorm/Abilities/Overpower/Camera/PlayerCameraModifier_ChapWorm_Overpower.PlayerCameraModifier_ChapWorm_Overpower_C"))
              (Variables|Transient|SetCameraModifier _returnvalue))
            (:else)))))))

(fn RemoveCameraModifier
  ;[0.009367 0.09587 0.1042] RemoveCameraMode: Remove the applied camera mode.
  (bind _cameramodifier (Variables|Transient|GetCameraModifier)
    (:then
      (CameraModifier|WindDownandRemove _cameramodifier :!class /Script/ChapWormGameplayRuntime.FortCameraModifier_ChapWorm))
    (:else)))
